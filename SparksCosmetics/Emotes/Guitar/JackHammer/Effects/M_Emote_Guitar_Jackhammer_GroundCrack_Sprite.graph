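# Translucent lit material for Niagara sprite and mesh particles, drawing the ground crack in the Jackhammer guitar emote.
# A packed map supplies erodible alpha, adjustable roughness and a crack mask that blends two tints and can darken opacity.
# Particle alpha drives erosion over the crack's lifetime, and particle color can optionally multiply the tint.
# Tinted, brightened diffuse passes through desaturation and a time-of-day color curve, alongside a flattenable normal map.
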Material /SparksCosmetics/Emotes/Guitar/JackHammer/Effects/M_Emote_Guitar_Jackhammer_GroundCrack_Sprite
graph {
 properties {
  bAutomaticallySetUsageInEditor=false
  bUsedWithNiagaraMeshParticles=true
  bUsedWithNiagaraSprites=true
  blendMode=BLEND_Translucent
  translucencyLightingMode=TLM_Surface
  usedShadingModels=Default Lit
 }
 nodes [
  n1 = TextureSampleParameter2D() {parameterName=AlphaRoughnessHightMap, texture=/Game/Effects/Fort_Effects/Textures/GroundCrack/T_Decal_GroundCrack_AlphaRH}
  Add(A=@n1.G, B=Divide(A=ScalarParameter() {parameterName=RoughnessCorrection}) {constB=5})
  Constant()
  Constant() {r=1}
  n2 = ParticleColor()
  n3 = Power(Base=@n2.A, Exp=ScalarParameter() {defaultValue=1, parameterName=DecalLifetimeOpacityPower})
  n4 = MaterialFunctionCall!(Dynamic Param=Multiply(A=OneMinus(None=@n3), B=ScalarParameter() {defaultValue=-1, parameterName=ErodeStrength}), Alpha=@n1.R) {materialFunction=/Game/Effects/Fort_Effects/Materials/_Functions/Erode_Alpha}
  n5 = Power(Base=Multiply(A=@n1.B, B=@n1.B), Exp=ScalarParameter() {defaultValue=1, parameterName=CrackMaskPower})
  LinearInterpolate(A=@n4.Result, B=Multiply(A=@n4.Result, B=OneMinus(None=@n5)), Alpha=ScalarParameter() {parameterName=CrackMaskStrength})
  n6 = TextureSampleParameter2D() {parameterName=NormalMap, samplerType=SAMPLERTYPE_Normal, texture=/Game/Effects/Fort_Effects/Textures/GroundCrack/T_Decal_GroundCrack_N}
  MaterialFunctionCall(Normal=@n6.RGB, Flatness=ScalarParameter() {defaultValue=0.2, parameterName=FlattenNormal}) {materialFunction=&FlattenNormal}
  n7 = Multiply(
   A=TextureSampleParameter2D() {parameterName=DiffuseMap, texture=/Game/Effects/Fort_Effects/Textures/GroundCrack/T_Decal_GroundCrack_D}.RGB
   B=ScalarParameter() {defaultValue=3, parameterName=DiffuseColorBrightness}
  )
  n8 = LinearInterpolate(
   A=VectorParameter() {parameterName=DecalColor, defaultValue=<rgba 0.5417 0.25 0.0616 1>}.RGB
   B=VectorParameter() {parameterName=DecalColorB, defaultValue=<rgba 1 0.6059 0.2209 1>}.RGB
   Alpha=@n1.B
  )
  n9 = LinearInterpolate(A=@n8, B=Multiply(A=@n8, B=@n2.RGB), Alpha=ScalarParameter() {parameterName=UseDecalColorInputData})
  Multiply(
   A=Desaturation(None=Multiply(A=@n7, B=@n9), Fraction=ScalarParameter() {parameterName=Desatuation})
   B=MaterialFunctionCall() {materialFunction=/Game/Effects/Fort_Effects/Materials/_Functions/CurveColor_Time_of_Day_S23}.Result
  )
 ]
}
