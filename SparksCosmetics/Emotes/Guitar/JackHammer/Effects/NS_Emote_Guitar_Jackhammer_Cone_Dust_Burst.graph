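# Niagara system for the Sparks guitar Jackhammer emote that emits a small one-shot puff of brownish dust sprites, named as a cone dust burst.
# A single Dust emitter spawns a few sprites spread by index around a small ring, with slow initial velocity, drag and rotation.
# Curves grow size, fade alpha and drive dynamic material parameters alongside SubUV animation.
# The User.bIsFrontEndPreview flag selects a more transparent color range for front-end preview.
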NiagaraSystem /SparksCosmetics/Emotes/Guitar/JackHammer/Effects/NS_Emote_Guitar_Jackhammer_Cone_Dust_Burst
dependencies {
 usedDataInterfaces [&NiagaraDataInterfaceCurve, &NiagaraDataInterfaceSpriteRendererInfo]
 usedDynamicInputs [&Vector2DFromFloat]
 usedModules [&SystemState]
}
emitters {Dust {
  emitterUpdateScript [
   &EmitterState()
   &SpawnBurst_Instantaneous(Spawn Count=6, Use Spawn Probability=false, UseLoopCountLimit=true)
  ]
  particleSpawnScript [
   &InitializeParticle(
    Color=&MakeCustomLinearColorFromBool(
     Bool=$User.bIsFrontEndPreview
     True Color=&RandomRangeLinearColor(Minimum=<rgba 0.3177 0.1972 0.1605 0.15>, Maximum=<rgba 0.3021 0.2182 0.1731 0.3>)
     False Color=&RandomRangeLinearColor(Minimum=<rgba 0.375 0.2073 0.1562 0.35>, Maximum=<rgba 0.375 0.2328 0.1562 0.5>)
    )
    Lifetime=&RandomRangeFloat(Minimum=0.55, Maximum=0.8)
    Sprite Rotation Mode=::NewEnumerator2
    Sprite Size Mode=::NewEnumerator1
    Uniform Sprite Size Max=25
    Uniform Sprite Size Min=15
    UsePositionOffset=false
    Write Material Random=true
   )
   &ShapeLocation(
    MaskSpawnGroup=false
    Override Seed=false
    Ring / Disc Distribution Mode:ENiagaraSphereDistributionMode=::NewEnumerator1
    Ring Radius=7
    Shape Primitive=::NewEnumerator4
    U Position=&ReturnNormalizedExecIndex(Normalized Index Includes One=false)
    UseOwnerScale=true
   ) {module=ShapeLocation001}
   &AddVelocity(
    Module.Use Falloff Distance=false
    Override Seed=false
    Velocity Mode=::NewEnumerator1
    Velocity Speed=&RandomRangeFloat(Minimum=10, Maximum=35)
   )
   &ApplyInitialForces()
  ]
  particleUpdateScript [
   &ParticleState()
   &SpriteRotationRate(Rotation Rate=&RandomRangeFloat(Minimum=3, Maximum=6))
   &ScaleColor(
    Scale Alpha=&FloatFromCurve(FloatCurve=DataInterface() {propertyValues {Curve {keys [
         {arriveTangent=0.7573, interpMode=RCIM_Cubic, leaveTangent=0.7573, tangentMode=RCTM_User, time=0, value=0}
         {arriveTangent=-0.0943, interpMode=RCIM_Cubic, leaveTangent=-0.0943, tangentMode=RCTM_User, time=0.1, value=1}
         {arriveTangent=-3.0484, interpMode=RCIM_Cubic, leaveTangent=-3.0484, tangentMode=RCTM_User, time=1, value=0}
        ]}}})
    ScaleA=true
    ScaleRGB=true
   )
   &Drag(Drag=&RandomRangeFloat(Minimum=0.5, Maximum=1), Use Linear Drag=true, Use Rotational Drag=true)
   &SolveForcesAndVelocity(Clamp Velocity=false, Limit Acceleration=false)
   &ScaleSpriteSize(Uniform Curve Sprite Scale=DataInterface() {propertyValues {Curve {keys [
        {arriveTangent=2.4019, interpMode=RCIM_Cubic, leaveTangent=2.4019, tangentMode=RCTM_User, time=0, value=0}
        {arriveTangent=0.0224, interpMode=RCIM_Cubic, leaveTangent=0.0224, tangentMode=RCTM_User, time=1, value=2.5}
       ]}}})
   &DynamicMaterialParameters(
    Index 0 Param 1=&FloatFromCurve(FloatCurve=DataInterface() {propertyValues {Curve {keys [
         {interpMode=RCIM_Cubic, tangentMode=RCTM_User, time=0, value=0.5}
         {arriveTangent=1.9611, interpMode=RCIM_Cubic, leaveTangent=1.9611, tangentMode=RCTM_User, time=0.1, value=0.1}
         {interpMode=RCIM_Cubic, tangentMode=RCTM_User, time=1, value=1}
        ]}}})
    Index 0 Param 2=&FloatFromCurve(FloatCurve=DataInterface() {propertyValues {Curve {keys [
         {interpMode=RCIM_Cubic, tangentMode=RCTM_User, time=0, value=1}
         {interpMode=RCIM_Cubic, tangentMode=RCTM_User, time=0.25, value=0.5}
         {interpMode=RCIM_Cubic, tangentMode=RCTM_User, time=1, value=1}
        ]}}})
    Index 0 Param 3=1
    Index 0 Param 4=&FloatFromCurve(FloatCurve=DataInterface() {propertyValues {Curve {keys [{tangentMode=RCTM_User, time=0, value=0}, {tangentMode=RCTM_User, time=1, value=0.3}]}}})
    Module.Index0Param0Write=true
    Module.Index0Param1Write=true
    Module.Index0Param2Write=true
    Module.Index0Param3Write=true
   )
   &SubUVAnimation(
    "Sprite Renderer "=DataInterface() {propertyValues {SpriteRenderer=&self:Dust_0.NiagaraSpriteRendererProperties_0}}
    SubUV Animation Mode:ENiagara_SubUVLookupModeV2=::NewEnumerator3
    UseEndFrame=false
    UsePlayRateScale=true
    UseStartFrame=false
   )
  ]
  rendererClasses [&NiagaraSpriteRendererProperties]
  renderers [{rendererClass=&NiagaraSpriteRendererProperties}]
 }}
summary {
 systemName=NS_Emote_Guitar_Jackhammer_Cone_Dust_Burst
 emitters [{emitterName=Dust, rendererClasses [&NiagaraSpriteRendererProperties]}]
 userVariables [User.bIsFrontEndPreview {type=NiagaraBool, defaultValue=true}]
}
