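# Looping Niagara dust effect for the Sparks guitar Jackhammer emote, built from one sprite emitter that continuously spawns soft, tinted puffs.
# User.bIsFrontEndPreview switches between two random color ranges, separating front-end preview from the other case.
# Puffs get mostly sideways velocity, then drift with curl noise, drag and slight upward force while growing, rotating either way and fading.
# SubUV flipbook animation and curve-driven dynamic material parameters vary each sprite over its life.
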NiagaraSystem /SparksCosmetics/Emotes/Guitar/JackHammer/Effects/NS_Emote_Guitar_Jackhammer_Dust_Looping
dependencies {
 usedDataInterfaces [
  &NiagaraDataInterfaceArrayFloat
  &NiagaraDataInterfaceCurve
  &NiagaraDataInterfaceSpriteRendererInfo
 ]
 usedDynamicInputs [&Vector2DFromFloat]
 usedModules [&SystemState]
}
emitters {Dust {
  emitterUpdateScript [
   &EmitterState()
   &SpawnRate(SpawnRate=6, Use Spawn Probability=false)
  ]
  particleSpawnScript [
   &InitializeParticle(
    Color=&MakeCustomLinearColorFromBool(
     Bool=$User.bIsFrontEndPreview
     True Color=&RandomRangeLinearColor(Minimum=<rgba 0.3177 0.1972 0.1605 0.35>, Maximum=<rgba 0.3021 0.2182 0.1731 0.5>)
     False Color=&RandomRangeLinearColor(Minimum=<rgba 0.375 0.2073 0.1562 0.35>, Maximum=<rgba 0.375 0.2328 0.1562 0.5>)
    )
    Lifetime=1.5
    Sprite Rotation Angle=&RandomRangeFloat(Minimum=-25, Maximum=25)
    Sprite Rotation Mode=::NewEnumerator0
    Sprite Size Mode=::NewEnumerator1
    Uniform Sprite Size Max=125
    Uniform Sprite Size Min=75
    UsePositionOffset=false
    Write Material Random=true
   )
   &AddVelocity(Velocity=&RandomRangeVector(Minimum=<-150 -20 0>, Maximum=<150 20 0>))
   &CameraOffset(Camera Offset Amount=&RandomRangeFloat(Minimum=-40, Maximum=40))
  ]
  particleUpdateScript [
   &ParticleState()
   &SpriteRotationRate(Rotation Rate=&Multiply_Float(
     A=&RandomRangeFloat(Minimum=6, Maximum=12)
     B=&SelectFloatFromArray(
      Array Sampling Mode=::NewEnumerator0
      Direct Array Index=?
      Float Selection Array=DataInterface() {propertyValues {FloatData [1,-1]}}
      Random Seed=0
     )
    ))
   &ScaleColor(
    Scale Alpha=&FloatFromCurve(FloatCurve=DataInterface() {propertyValues {Curve {keys [
         {arriveTangent=0.7573, interpMode=RCIM_Cubic, leaveTangent=0.7573, tangentMode=RCTM_User, time=0, value=0}
         {arriveTangent=-0.009, interpMode=RCIM_Cubic, leaveTangent=-0.009, tangentMode=RCTM_User, time=0.1, value=1}
         {arriveTangent=-3.2549, interpMode=RCIM_Cubic, leaveTangent=-3.2549, tangentMode=RCTM_User, time=1, value=0}
        ]}}})
    ScaleA=true
    ScaleRGB=true
   )
   /Niagara/Modules/Update/Forces/CurlNoiseForce(Noise Frequency=20, Noise Strength=50, Pan Noise=true, Pan Noise Field=<0 0 1>)
   &Drag(Drag=&RandomRangeFloat(Minimum=2, Maximum=4), Use Linear Drag=true, Use Rotational Drag=true)
   &GravityForce(Gravity=&RandomRangeVector(Minimum=<-30 -30 10>, Maximum=<30 30 30>))
   &SolveForcesAndVelocity(Clamp Velocity=false, Limit Acceleration=false)
   &ScaleSpriteSize(Uniform Curve Sprite Scale=DataInterface() {propertyValues {Curve {keys [
        {arriveTangent=3.7064, interpMode=RCIM_Cubic, leaveTangent=3.7064, tangentMode=RCTM_User, time=0, value=0.375}
        {arriveTangent=-0.0206, interpMode=RCIM_Cubic, leaveTangent=-0.0206, tangentMode=RCTM_User, time=1, value=2}
       ]}}})
   &DynamicMaterialParameters(
    Index 0 Param 1=&FloatFromCurve(FloatCurve=DataInterface() {propertyValues {Curve {keys [
         {arriveTangent=-0.0075, interpMode=RCIM_Cubic, leaveTangent=-0.0075, tangentMode=RCTM_User, time=0.7, value=1}
         {arriveTangent=-5.6922, interpMode=RCIM_Cubic, leaveTangent=-5.6922, tangentMode=RCTM_User, time=1, value=0.075}
        ]}}})
    Index 0 Param 2=&FloatFromCurve(FloatCurve=DataInterface() {propertyValues {Curve {keys [
         {interpMode=RCIM_Cubic, tangentMode=RCTM_User, time=0, value=1}
         {interpMode=RCIM_Cubic, tangentMode=RCTM_User, time=0.25, value=0.5}
         {interpMode=RCIM_Cubic, tangentMode=RCTM_User, time=1, value=1}
        ]}}})
    Index 0 Param 3=1
    Module.Index0Param0Write=true
    Module.Index0Param1Write=true
    Module.Index0Param2Write=true
    Module.Index0Param3Write=true
   )
   &SubUVAnimation(
    Frames Per Second:NiagaraInt32=&RandomRangeInt(Minimum=10, Maximum=12)
    Playback Mode:ENiagara_FPSPlayrate=::NewEnumerator1
    Playback Scale Curve=DataInterface() {propertyValues {Curve {keys [
        {interpMode=RCIM_Cubic, tangentMode=RCTM_User, time=0.25, value=1}
        {arriveTangent=-1.2656, interpMode=RCIM_Cubic, leaveTangent=-1.2656, tangentMode=RCTM_User, time=1, value=0}
       ]}}}
    Random Start Frame=true
    "Sprite Renderer "=DataInterface() {propertyValues {SpriteRenderer=&self:Dust_0.NiagaraSpriteRendererProperties_0}}
    Start Frame Offset:NiagaraInt32=&RandomRangeInt(Minimum=0, Maximum=16)
    SubUV Animation Mode:ENiagara_SubUVLookupModeV2=::NewEnumerator3
    UseEndFrame=true
    UsePlayRateScale=false
    UseStartFrame=true
   )
  ]
  rendererClasses [&NiagaraSpriteRendererProperties]
  renderers [{rendererClass=&NiagaraSpriteRendererProperties}]
 }}
summary {
 systemName=NS_Emote_Guitar_Jackhammer_Dust_Looping
 emitters [{emitterName=Dust, rendererClasses [&NiagaraSpriteRendererProperties]}]
 userVariables [User.bIsFrontEndPreview {type=NiagaraBool}]
}
