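# Niagara system for the Jackhammer guitar emote that bursts short-lived ground cracks with random size and rotation.
# One emitter carries both an upward-facing sprite renderer and a decal renderer, and sets a particle visibility tag from User.bIsFrontEndPreview, apparently to choose between them.
# Matching curves snap sprite and decal size to full almost instantly, then hold opacity before fading out late in the particle's life.
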NiagaraSystem /SparksCosmetics/Emotes/Guitar/JackHammer/Effects/NS_Emote_Guitar_Jackhammer_GroundCracks
dependencies {usedDataInterfaces [&NiagaraDataInterfaceCurve], usedDynamicInputs [], usedModules [&SystemState]}
emitters {Ground_Cracks {
  emitterUpdateScript [
   &EmitterState()
   &SpawnBurst_Instantaneous(Use Spawn Probability=false, UseLoopCountLimit=false)
  ]
  particleSpawnScript [
   &InitializeParticle(
    Lifetime=4.5
    Sprite Rotation Mode=::NewEnumerator2
    Sprite Size Mode=::NewEnumerator1
    Uniform Sprite Size Max=130
    Uniform Sprite Size Min=90
    UsePositionOffset=true
    Write Material Random=true
   ) {module=InitializeParticle001}
   /SetVariables(Particles.RandomDecalYZ:NiagaraFloat=&RandomRangeFloat(Minimum=90, Maximum=130)) {isSetParameters=true}
   /SetVariables(Particles.DecalSize:Vector3f=&MakeVector(X=75, Y=$Particles.RandomDecalYZ, Z=$Particles.RandomDecalYZ)) {isSetParameters=true}
   &InitialMeshOrientation(
    Mesh Orientation Mode=::NewEnumerator2
    Orientation Vector=<0 0 1>
    Rotation=&RandomRangeVector(Minimum=<-1 0 0>, Maximum=<1 0 0>)
    Use Rotation Vector=true
   )
   /SetVariables(Particles.VisibilityTag=&MakeCustomIntFromBool(Bool=$User.bIsFrontEndPreview, True INT=0, False INT=1)) {isSetParameters=true}
   &SpriteFacingAndAlignment(Sprite Alignment=<1 0 0>, Sprite Facing=<0 0 1>, WriteAlignment=true, WriteFacing=true) {module=SpriteFacingAndAlignment001}
  ]
  particleUpdateScript [
   &ParticleState()
   &ScaleSpriteSize(Uniform Curve Sprite Scale=DataInterface() {propertyValues {Curve {keys [
        {arriveTangent=230.8238, interpMode=RCIM_Cubic, leaveTangent=230.8238, tangentMode=RCTM_User, time=0, value=0}
        {arriveTangent=0.8135, interpMode=RCIM_Cubic, leaveTangent=0.8135, tangentMode=RCTM_User, time=0.01, value=1}
       ]}}})
   &ScaleColor(
    Scale Alpha=&FloatFromCurve(FloatCurve=DataInterface() {propertyValues {Curve {keys [
         {interpMode=RCIM_Cubic, tangentMode=RCTM_User, time=0, value=0}
         <at 0.025 1>
         {arriveTangent=-1.0256, interpMode=RCIM_Cubic, leaveTangent=-1.0256, time=0.7, value=1}
         <at 1 0 RCIM_Cubic>
        ]}}})
    ScaleA=true
    ScaleRGB=true
   )
   &ScaleSpriteSize(Uniform Curve Sprite Scale=DataInterface() {propertyValues {Curve {keys [{arriveTangent=139.099, interpMode=RCIM_Cubic, leaveTangent=139.099, tangentMode=RCTM_User, time=0, value=0}, <at 0.015 1 RCIM_Cubic>]}}}) {module=ScaleSpriteSize001}
   /SetVariables(Particles.DecalFade:NiagaraFloat=&FloatFromCurve(FloatCurve=DataInterface() {propertyValues {Curve {keys [
         {interpMode=RCIM_Cubic, tangentMode=RCTM_User, time=0, value=0}
         <at 0.025 1>
         {arriveTangent=-1.0256, interpMode=RCIM_Cubic, leaveTangent=-1.0256, time=0.7, value=1}
         <at 1 0 RCIM_Cubic>
        ]}}})) {isSetParameters=true}
   /SetVariables(Particles.DecalSize:Vector3f=&Multiply_VectorByFloat(
     Vector=$Particles.Initial.DecalSize
     Float=&FloatFromCurve(FloatCurve=DataInterface() {propertyValues {Curve {keys [{arriveTangent=139.099, interpMode=RCIM_Cubic, leaveTangent=139.099, tangentMode=RCTM_User, time=0, value=0}, <at 0.015 1 RCIM_Cubic>]}}})
    )) {isSetParameters=true}
  ]
  rendererClasses [&NiagaraSpriteRendererProperties, &NiagaraDecalRendererProperties]
  renderers [
   {rendererClass=&NiagaraSpriteRendererProperties}
   {rendererClass=&NiagaraDecalRendererProperties, rendererIndex=1}
  ]
 }}
summary {
 systemName=NS_Emote_Guitar_Jackhammer_GroundCracks
 emitters [{emitterName=Ground_Cracks, rendererClasses [&NiagaraSpriteRendererProperties, &NiagaraDecalRendererProperties]}]
 userVariables [
  User.bIsFrontEnd {type=NiagaraBool}
  User.bIsFrontEndPreview {type=NiagaraBool}
 ]
}
